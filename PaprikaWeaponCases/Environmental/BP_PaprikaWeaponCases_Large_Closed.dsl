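; Closed large weapon case environmental actor blueprint for the Paprika weapon cases set.
; When its linked actor is destroyed, the FortLinkToActor event sets this actor's life span to a short delay so it removes itself shortly after.
; The construction script is empty.
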(event OnLinkedActorDestroyed(FortLinkToActor)
  (params (DamageCauser :type Actor%20Object%20Reference))
  (Actor|SetLifeSpan self 0.25))

(fn ConstructionScript)
